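# Fort_Events_Mapwide is a sound class for map-wide event audio, parented to the Fort_Events_Master class.
# It carries a passive sound mix modifier that applies the SM_Fort_Events_Mapwide_PSM mix while sounds in this class play above a low minimum volume threshold.
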SoundClass /Figment_S08/Athena/Sounds/Events/SoundClasses/Fort_Events_Mapwide
parent=&/Fort_Events_Master
properties {passiveSoundMixModifiers [{minVolumeThreshold=0.15, soundMix=&/SM_Fort_Events_Mapwide_PSM}]}
