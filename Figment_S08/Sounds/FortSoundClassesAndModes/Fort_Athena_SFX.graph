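# Sound class grouping the Athena (battle royale) in-game sound effects in the Figment_S08 content.
# It sits under Fort_SFX_InGame and parents the classes for Athena player foley, consumables and the events master, with two child references named Fort_Consumables_Athena.
# Its own properties set a partial LFE bleed for sounds assigned to it.
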SoundClass /Figment_S08/Sounds/FortSoundClassesAndModes/Fort_Athena_SFX
parent=&Fort_SFX_InGame
children [
 &Fort_Player_Foley_Athena
 &/Fort_Consumables_Athena
 /Figment_S08/Athena/Sounds/Events/SoundClasses/Fort_Events_Master
 &Fort_Consumables_Athena
]
properties {properties {lFEBleed=0.25}}
