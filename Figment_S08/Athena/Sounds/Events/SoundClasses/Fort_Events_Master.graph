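# Fort_Events_Master is a Sound Class that groups event audio for the Athena mode in the Figment_S08 content.
# It parents to Fort_Athena_SFX, placing events under the game's SFX class hierarchy.
# Its child classes split event sounds into four priority tiers from extra high to low, plus dedicated Festivus and mapwide classes.
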SoundClass /Figment_S08/Athena/Sounds/Events/SoundClasses/Fort_Events_Master
parent=/Figment_S08/Sounds/FortSoundClassesAndModes/Fort_Athena_SFX
children [
 &/Fort_Events_Priority_ExtHigh
 &/Fort_Events_Priority_High
 &/Fort_Events_Priority_Medium
 &/Fort_Events_Priority_Low
 &/Fort_Events_Festivus
 &/Fort_Events_Mapwide
]
