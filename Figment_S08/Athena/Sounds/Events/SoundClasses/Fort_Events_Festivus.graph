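# Sound class for Festivus event audio in Figment Season 8 Athena.
# It sits under the Fort_Events_Master class and parents a Fort_Events_Festivus_PIESuppress child class, whose name suggests suppression during Play In Editor.
# It boosts volume above unity and flags its sounds to always play, so they take priority over others when voices are limited.
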SoundClass /Figment_S08/Athena/Sounds/Events/SoundClasses/Fort_Events_Festivus
parent=&/Fort_Events_Master
children [
 &/Fort_Events_Festivus_PIESuppress
 &Fort_Events_Festivus_PIESuppress
]
properties {properties {bAlwaysPlay=true, volume=1.6}}
